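# Post-process AnimBlueprint for the large male SlopeTramp player head, layering procedural adjustments and secondary ear motion over the input pose.
# A Control Rig processes the input into a cached TargetPose.
# When bIsTargeting is true, it blends to a version with additive translation offsets on upper spine and clavicle bones.
# Chained AnimDynamics nodes then simulate both din_ears bones with angular springs and limits; the EventGraph has no logic.
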AnimBlueprint /BRCosmetics/Characters/Player/Male/Large/Heads/M_LRG_SlopeTramp_Head/Meshes/M_LRG_SlopeTramp_Head_AnimBP_Postprocess
target_skeleton=&/M_LRG_SlopeTramp_Head_Skeleton
graphs [
 AnimGraph {nodes [
   n1 = "Transform (Modify) Bone - Bone: spine_05"!(ComponentPose=Local To Component(LocalPose=Use cached pose 'TargetPose'()), Alpha=1, Translation="-2,2,0", Rotation="0,0,0") -> (Pose) {
    rotationMode=BMM_Additive
    translationMode=BMM_Additive
    alphaBoolBlend {blendOption=Linear}
    alphaScaleBiasClamp {bInterpResult=true, interpSpeedDecreasing=5, interpSpeedIncreasing=5}
    boneToModify {boneName=spine_05}
    translation=<-2 2>
   }
   n2 = "Transform (Modify) Bone - Bone: spine_04"!(ComponentPose=@n1, Alpha=1, Translation="-1,2,0", Rotation="0,0,0") -> (Pose) {
    rotationMode=BMM_Additive
    translationMode=BMM_Additive
    alphaBoolBlend {blendOption=Linear}
    alphaScaleBiasClamp {bInterpResult=true, interpSpeedDecreasing=5, interpSpeedIncreasing=5}
    boneToModify {boneName=spine_04}
    translation=<-1 2>
   }
   n3 = "Transform (Modify) Bone - Bone: clavicle_l"!(ComponentPose=@n2, Alpha=1, Translation="3,-4,0", Rotation="0,0,0") -> (Pose) {
    rotationMode=BMM_Additive
    translationMode=BMM_Additive
    alphaBoolBlend {blendOption=Linear}
    alphaScaleBiasClamp {bInterpResult=true, interpSpeedDecreasing=5, interpSpeedIncreasing=5}
    boneToModify {boneName=clavicle_l}
    translation=<3 -4>
   }
   n4 = "Transform (Modify) Bone - Bone: clavicle_r"!(ComponentPose=@n3, Alpha=1, Translation="3,-4,0", Rotation="0,0,0") -> (Pose) {
    rotationMode=BMM_Additive
    translationMode=BMM_Additive
    alphaBoolBlend {blendOption=Linear}
    alphaScaleBiasClamp {bInterpResult=true, interpSpeedDecreasing=5, interpSpeedIncreasing=5}
    boneToModify {boneName=clavicle_r}
    translation=<3 -4>
   }
   n5 = Component To Local(ComponentPose=@n4)
   n6 = Use cached pose 'TargetPose'()
   n7 = Get bIsTargeting()
   n8 = Anim Dynamics - Bone: din_ears_l!(
    SimulationSpace:AnimPhysSimSpaceType Enum=Component
    ComponentPose=Local To Component(LocalPose=Blend Poses by bool(BlendPose_0=@n5, BlendPose_1=@n6, bActiveValue=@n7) {blendType=Linear})
    Alpha=1
    ExternalForce="0,0,0"
   ) -> (Pose) {
    angularDampingOverride=0.9999
    angularSpringConstant=500
    bAngularSpring=true
    bOverrideAngularDamping=true
    bUseGravityOverride=true
    linearDampingOverride=0.9999
    simulationSpace=World
    alphaBoolBlend {blendOption=Linear}
    boundBone {boneName=din_ears_l}
    componentAppliedLinearAccClamp=<3000 3000 3000>
    componentLinearAccScale {z=0.7}
    componentLinearVelScale {z=0.7}
    physicsBodyDefinitions [{
      boundBone {boneName=din_ears_l}
      boxExtents=<10 10 10>
      constraintSetup {angularLimitsMax=<0 30 0>, angularLimitsMin=<0 0 0>, angularTarget=<1 0 0>}
      localJointOffset=<0.2 0 0>
     }]
    simSpaceSettings {maxAngularAcceleration=600, maxAngularVelocity=300, simSpaceAngularAlpha=1}
   }
   n9 = Anim Dynamics - Bone: din_ears_r!(SimulationSpace:AnimPhysSimSpaceType Enum=Component, ComponentPose=@n8, Alpha=1, ExternalForce="0,0,0") -> (Pose) {
    angularDampingOverride=0.9999
    angularSpringConstant=500
    bAngularSpring=true
    bOverrideAngularDamping=true
    bUseGravityOverride=true
    linearDampingOverride=0.9999
    simulationSpace=World
    alphaBoolBlend {blendOption=Linear}
    boundBone {boneName=din_ears_r}
    componentAppliedLinearAccClamp=<3000 3000 3000>
    componentLinearAccScale {z=0.7}
    componentLinearVelScale {z=0.7}
    physicsBodyDefinitions [{
      boundBone {boneName=din_ears_r}
      boxExtents=<10 10 10>
      constraintSetup {angularLimitsMax=<0 0 0>, angularLimitsMin=<0 -30 0>, angularTarget=<1 0 0>}
      localJointOffset=<-0.2 0 0>
     }]
    simSpaceSettings {maxAngularAcceleration=600, maxAngularVelocity=300, simSpaceAngularAlpha=1}
   }
   Output Pose(Result=Component To Local(ComponentPose=@n9))
   n10 = Control Rig(Source=Input Pose()) {controlRigAssetReference {blueprintRigClass=&/M_LRG_SlopeTramp_Head_CtrlRig.M_LRG_SlopeTramp_Head_CtrlRig_C}}
   Save cached pose 'TargetPose'(Pose=@n10) {cacheName=TargetPose}
  ]}
 EventGraph {nodes [
   TryGetPawnOwner()
   Event BlueprintUpdateAnimation()
  ]}
]
variables []
